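; Base Creative device for a resizable effect volume that filters by team and class, leaving enter/exit handlers empty for children to override.
; Trigger overlaps sort actors into vehicles, players, creatures, bots or others.
; Minigame start can equip an ability set and bind team, class and spawn events that revalidate overlapping pawns; registered players also get vehicle enter/exit handling.
; Enabled state and play mode drive trigger collision and visibility.
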(event EventOnWorldReady
  ;[1 0.07005 0.7112] Begin Play
  (|CallParentFunction)
  (CallFunction|UpdateBaseVisibilityComponent))

(event Custom|EnableOnReceived
  (params (Instigator :type Controller%20Object%20Reference))
  ;[1 1 1] Channel-based activation logic
  (Class|CreativeEnabledComponent|SetEnabled (Variables|Default|GetEnabledComponent) true))

(event EventInitializeDevice
  ;[1 0.07005 0.7112] Begin Play
  (|CallParentFunction)
  (CallFunction|UpdateBaseVisibilityComponent)
  (CallFunction|UpdateVolumeSize)
  (Initialization|InitializeComponentWithAffects (Variables|Default|GetCreativeUseRestrictionComponent) (Variables|Default|GetFortMinigameProgress) 0 0 (Variables|DeviceOptions|GetZoneProtectedTeam) (Variables|DeviceOptions|GetKnob_SelectedClass) 0 1 (not (Math|Conversions|ToBoolean(Integer) (Variables|DeviceOptions|GetKnob_AffectsTeam))) (not (Math|Conversions|ToBoolean(Integer) (Variables|DeviceOptions|GetKnob_AffectsClass))) (Variables|Default|GetUses%20Safe%20Team%20Indexing))
  (Resizing|UpdateVolumeState))

(event Custom|DisableOnReceived
  (params (Instigator :type Controller%20Object%20Reference))
  ;[1 1 1] Channel-based activation logic
  (Class|CreativeEnabledComponent|SetEnabled (Variables|Default|GetEnabledComponent) false))

(event Custom|OnPlayerClassApplied
  (params (PlayerState :type Fort%20Player%20State%20Object%20Reference) (ClassSlotIndex :type Byte))
  ;[0.3833 0.4599 1] Handle Options Accessible During Gameplay
  (Registration|PlayerOptionChangedDuringGame PlayerState))

(event ChildFunctions|EventPlayModeChanged
  (params (Minigame :type Fort%20Minigame%20Object%20Reference) (bIsInPlayMode :type Boolean))
  (|CallParentFunction Minigame bIsInPlayMode)
  (Resizing|UpdateVolumeState))

(event Custom|OnPlayerChangedTeam
  (params (PlayerState :type Fort%20Player%20State%20Object%20Reference) (TeamId :type Byte) (OldTeamId :type Byte) (bTeamChangedWithoutRespawn :type Boolean))
  ;[0.3833 0.4599 1] Handle Options Accessible During Gameplay
  (Registration|PlayerOptionChangedDuringGame PlayerState))

(event Custom|OnMinigamePlayerSpawned
  (params (SpawningPlayer :type Controller%20Object%20Reference))
  (bind _returnvalue (Pawn|GetControlledPawn SpawningPlayer))
  ;[0.3833 0.4599 1] Handle Options Accessible During Gameplay
  ;;[0.451 0.451 0.451 0.25] If the device is disabled, there is no overlap
  (Utilities|IsValid _returnvalue
    (:Is%20Valid
      (if (Collision|IsOverlappingActor (Variables|Default|GetTrigger) _returnvalue :!class /Script/Engine.PrimitiveComponent)
        (Registration|RevalidatePawn _returnvalue)))))

(event MinigameStartedDelegate(FortMinigameProgress)
  ;[0.9809 1 0.2776] Minigame State Changes (Begin and End, Transitionals)
  (if (Variables|Default|GetSetup%20at%20Minigame%20Start/End)
    (CallFunction|SetupDelegatesAndAbilities true)))

(event MinigameEndedDelegate(FortMinigameProgress)
  ;[0.9809 1 0.2776] Minigame State Changes (Begin and End, Transitionals)
  (if (Variables|Default|GetSetup%20at%20Minigame%20Start/End)
    (CallFunction|SetupDelegatesAndAbilities false)))

(event Custom|Setup%20Delegates%20and%20Abilities
  (params (Setup%20 :type Boolean))
  ;[0.9809 1 0.2776] Minigame State Changes (Begin and End, Transitionals)
  (if Setup%20
    (Networking|SwitchHasAuthority
      (:Authority
        (UserOptions|CustomInitializeClass (Variables|DeviceOptions|GetKnob_SelectedClass))
        (UserOptions|CustomInitializeTeams (Variables|DeviceOptions|GetZoneProtectedTeam))
        (bind _currentminigame (Default|GetCurrentMinigame)
          (:Is%20Valid
            (Minigame|BindEventtoOnPlayerClassApplied _currentminigame (EventDispatchers|CreateEvent :!event Custom|OnPlayerClassApplied))
            (Minigame|BindEventtoPlayerChangedTeamDelegate _currentminigame (EventDispatchers|CreateEvent :!event Custom|OnPlayerChangedTeam))
            (Minigame|AssignOnMinigamePlayerSpawned _currentminigame (EventDispatchers|CreateEvent :!event Custom|OnMinigamePlayerSpawned))
            (bind _returnvalue (Ability|EquipFortAbilitySet self (Variables|Ability|GetAbilitySystemReference)))
            (Variables|Ability|SetAbilitySetHandle _returnvalue)
            (CallFunction|InitializeDevice)))))
    (else
      (Networking|SwitchHasAuthority
        (:Authority
          (bind _currentminigame_2 (Default|GetCurrentMinigame)
            (:Is%20Valid
              (Minigame|UnbindEventfromOnPlayerClassApplied _currentminigame_2 (EventDispatchers|CreateEvent :!event Custom|OnPlayerClassApplied))
              (Minigame|UnbindEventfromPlayerChangedTeamDelegate _currentminigame_2 (EventDispatchers|CreateEvent :!event Custom|OnPlayerChangedTeam))
              (Minigame|UnbindEventfromOnMinigamePlayerSpawned _currentminigame_2 (EventDispatchers|CreateEvent :!event Custom|OnMinigamePlayerSpawned))
              (Ability|UnequipFortAbilitySet (Variables|Ability|GetAbilitySetHandle)))))))))

(event Custom|Pawn%20Entered%20Trigger%20Volume
  (params (Pawn :type Pawn%20Object%20Reference) (PlayerController :type Player%20Controller%20Object%20Reference)))

(event OnEnabledStateChanged(EnabledComponent)
  (params (Enabled :type Boolean))
  ;[1 0.3146 0.8498] On Enable
  (Resizing|UpdateVolumeState))

(event OnComponentEndOverlap(Trigger)
  (params (OverlappedComponent :type Primitive%20Component%20Object%20Reference) (OtherActor :type Actor%20Object%20Reference) (OtherComp :type Primitive%20Component%20Object%20Reference) (OtherBodyIndex :type Integer))
  ;[1 1 1] Handle Overlap Events (Start and End)
  (CallFunction|OverlapEvent true OtherActor))

(event Custom|Pawn%20Exited%20Trigger%20Volume
  (params (Pawn :type Pawn%20Object%20Reference) (PlayerController :type Player%20Controller%20Object%20Reference)))

(event OnComponentBeginOverlap(Trigger)
  (params (OverlappedComponent :type Primitive%20Component%20Object%20Reference) (OtherActor :type Actor%20Object%20Reference) (OtherComp :type Primitive%20Component%20Object%20Reference) (OtherBodyIndex :type Integer) (bFromSweep :type Boolean) (SweepResult :type Hit%20Result%20Structure%20%28by%20ref%29))
  ;[1 1 1] Handle Overlap Events (Start and End)
  (CallFunction|OverlapEvent false OtherActor))

(event Custom|Non-Pawn%20Entered%20Trigger%20Volume
  (params (Actor :type Actor%20Object%20Reference)))

(event OnMeshandMaterialsLoaded(CreativeEditOnlyMesh)
  (CallFunction|ResizeFrame))

(event Custom|Non-Pawn%20Exited%20Trigger%20Volume
  (params (Actor :type Actor%20Object%20Reference)))

(event Custom|Player%20Bot%20Entered%20Trigger%20Volume
  (params (Pawn :type Pawn%20Object%20Reference) (AI%20Bot%20Controller :type Fort%20Athena%20AIBot%20Controller%20Object%20Reference)))

(event Custom|Player%20Bot%20Exited%20Trigger%20Volume
  (params (Pawn :type Pawn%20Object%20Reference) (AI%20Bot%20Controller :type Fort%20Athena%20AIBot%20Controller%20Object%20Reference)))

(event Custom|OverlapEvent
  (params (Has%20Ended%20Overlap :type Boolean) (Other%20Actor :type Actor%20Object%20Reference))
  (bind (_true _pawn) (Default|ActorIsACreature Other%20Actor))
  (bind (_true_2 _pawn_2 _ai_bot_controller) (Default|ActorIsAPlayerBot Other%20Actor))
  ;[1 0.6649 0.643] At the most basic, we check vehicles, then players, then creatures.  We keep this as an event in case we need to override it for children
  (Utilities|Casting|CastToFortVehicleInterface Other%20Actor
    (:then
      (if Has%20Ended%20Overlap
        (CallFunction|Non-PawnExitedTriggerVolume Other%20Actor)
        (else
          (CallFunction|Non-PawnEnteredTriggerVolume Other%20Actor))))
    (:CastFailed
      (bind (_valid_controller _fortplayerpawn _playercontroller) (DEPRECATED|ActorHasValidController (Variables|Default|GetCreativeUseRestrictionComponent) Other%20Actor true true))
      (if _valid_controller
        (if Has%20Ended%20Overlap
          (CallFunction|PawnExitedTriggerVolume _fortplayerpawn _playercontroller)
          (else
            (CallFunction|PawnEnteredTriggerVolume _fortplayerpawn _playercontroller)))
        (elif _true
          (if Has%20Ended%20Overlap
            (CallFunction|PawnExitedTriggerVolume _pawn)
            (else
              (CallFunction|PawnEnteredTriggerVolume _pawn)))
          (elif _true_2
            (if Has%20Ended%20Overlap
              (CallFunction|PlayerBotExitedTriggerVolume _pawn_2 _ai_bot_controller)
              (else
                (CallFunction|PlayerBotEnteredTriggerVolume _pawn_2 _ai_bot_controller)))
            (elif Has%20Ended%20Overlap
              (CallFunction|Non-PawnExitedTriggerVolume Other%20Actor)
              (else
                (CallFunction|Non-PawnEnteredTriggerVolume Other%20Actor)))))))))

(event Custom|UpdateVolumeSize
  ;[0.6552 0.6552 0.6552] Handle Volume Resizing.  Assumes the volume is a mesh
  (bind _output_get (Variables|Default|SetNewScale (Math|Vector|MakeVector (Variables|Default|GetZoneXScale) (Variables|Default|GetZoneYScale) (Variables|Default|GetZoneZScale))))
  (Transformation|SetRelativeScale3D (Variables|Default|GetZoneMesh) (* _output_get (Variables|Default|GetZoneMeshScalar)))
  (Transformation|SetRelativeScale3D (Variables|Default|GetTrigger) (Variables|Default|GetNewScale))
  (bind _location (Resizing|GetTriggerLocation))
  (Transformation|SetRelativeLocation (Variables|Default|GetTrigger) _location false true)
  (CallFunction|ResizeFrame))

(event PawnNowHasController(CreativeUseRestrictionComponent)
  (params (FortPlayerPawn :type Fort%20Player%20Pawn%20Object%20Reference) (Player%20Controller :type Fort%20Player%20Controller%20Object%20Reference))
  ;[1 1 1] Callback in case we are waiting on readiness
  (if (DEPRECATED|CanPlayerControllerInteract (Variables|Default|GetCreativeUseRestrictionComponent) Player%20Controller true false)
    (CallFunction|PawnEnteredTriggerVolume FortPlayerPawn Player%20Controller)))

(fn ActorIsACreature
  (params (Actor :type Actor%20Object%20Reference))
  (Utilities|Casting|CastToFortPlayerPawn Actor
    (:then
      (return false 0))
    (:CastFailed
      (bind _asfort_pawn (Utilities|Casting|CastToFortPawn Actor)
        (:then
          (return true _asfort_pawn))
        (:CastFailed
          (return false 0))))))

(fn ActorIsAPlayerBot
  (params (Actor :type Actor%20Object%20Reference))
  (bind _returnvalue (AI|GetAIController Actor))
  ;[1 1 1] Check if the actor has an AI controller and then determine if it is AI Bot Controller
  (Utilities|IsValid _returnvalue
    (:Is%20Valid
      (bind _asfort_athena_aibot_controller (Utilities|Casting|CastToFortAthenaAIBotController _returnvalue)
        (:then
          (return true (Variables|/Script/FortniteAI.FortAthenaAIBotController|GetPlayerBotPawn _asfort_athena_aibot_controller) _asfort_athena_aibot_controller))
        (:CastFailed
          _)))
    (:Is%20Not%20Valid
      _))
  (return false 0 0))

(fn AddAffectedPlayerInVolume
  (params (AffectedPlayer :type Actor%20Object%20Reference%20%28by%20ref%29))
  ;[1 1 1] We've got a new player in the volume
  (Utilities|IsValid AffectedPlayer
    (:Is%20Valid
      (bind _index (Registration|RegisterPlayerinVolume AffectedPlayer))
      (if (>= _index 0)
        (Vehicles|BindVehicleDelegates AffectedPlayer)))))

(fn AreOptionalsVisibleDuringMinigame
  ;[1 1 1] Get the current vis settings determined by users.
  (return (Visibility|ShouldShowinGame (Variables|DeviceOptions|GetBaseMeshVisibleInGame)) (Visibility|ShouldShowinGame (Variables|DeviceOptions|GetZoneVisibleInGame))))

(fn BindVehicleDelegates
  (params (Player :type Actor%20Object%20Reference))
  (bind _outputdelegate (EventDispatchers|CreateEvent :!event OnAffectedPlayerEnteredVehicle))
  ;[0.5222 1 0.4526] Special-case handling for entering/exiting vehicles while affected by a volume
  (bind _output_get (LocalVariables|SetPlayerController (Default|GetPlayerControllerfromActor Player)))
  (Utilities|IsValid _output_get
    (:Is%20Valid
      (Vehicle|BindEventtoOnExitVehicle (LocalVariables|GetPlayerController) (EventDispatchers|CreateEvent :!event OnAffectedPlayerExitedVehicle))
      (Vehicle|BindEventtoOnEnterVehicleDriver (LocalVariables|GetPlayerController) _outputdelegate)
      (Vehicle|BindEventtoOnEnterVehiclePassenger (LocalVariables|GetPlayerController) _outputdelegate)
      (Vehicles|OnAffectedPlayerEnteredVehicle))))

(fn CustomInitializeClass
  (params (SelectedClass :type Integer))
  (Networking|SwitchHasAuthority
    (:Authority
      (ClassRestrictions|SetClassWithAffects (Variables|Default|GetCreativeUseRestrictionComponent) SelectedClass (== (Variables|DeviceOptions|GetKnob_AffectsClass) (Variables|/Game/Creative/Devices/Common/Components/Creative_UseRestriction_Component.Creative_UseRestriction_Component_C|GetAffectsAllButSelectedValue (Variables|Default|GetCreativeUseRestrictionComponent))) (Variables|Default|GetUses%20Safe%20Team%20Indexing)))))

(fn CustomInitializeTeams
  (params (SelectedTeam :type Integer))
  (Networking|SwitchHasAuthority
    (:Authority
      (TeamRestrictions|SetTeamswithAffects (Variables|Default|GetCreativeUseRestrictionComponent) SelectedTeam (== (Variables|/Game/Creative/Devices/Common/Components/Creative_UseRestriction_Component.Creative_UseRestriction_Component_C|GetAffectsAllButSelectedValue (Variables|Default|GetCreativeUseRestrictionComponent)) (Variables|DeviceOptions|GetKnob_AffectsTeam))))))

(fn GetCreativeActorBounds
  (return false "0, 0, 0"))

(fn GetCreativeActorOrigin
  (return true (Math|Vector|MakeVector 0.0 0.0 (* (Variables|Default|GetZoneZScale) (Variables|Default|GetPivotOriginOffset)))))

(fn GetCreativeActorRelativeTransform
  (return false 0))

(fn GetGridAxisGuideAttachmentPoint
  (return false false "0, 0, 0"))

(fn GetOverrideMeshMaterial
  (params (MeshComp :type Static%20Mesh%20Component%20Object%20Reference) (MatIdx :type Integer))
  (if (Utilities|Equal(Object) MeshComp (Variables|Default|GetCreativeEditOnlyMesh))
    (return (Variables|Default|GetFrame%20MID))
    (else
      (return (|CallParentFunction MeshComp MatIdx)))))

(fn GetPlayerControllerFromActor
  (params (Actor :type Actor%20Object%20Reference))
  ;[1 0.7271 0.6488] Retrieve the player controller from an actor.  Assumes any input so might not be the best option in all cases.
  (LocalVariables|SetInActor Actor)
  (Utilities|FlowControl|Sequence
    (:then_0
      (bind _aspawn (Utilities|Casting|CastToPawn (LocalVariables|GetInActor)))
      (bind _asfort_player_controller (Utilities|Casting|CastToFortPlayerController (Pawn|GetController _aspawn))
        (:then
          (return _asfort_player_controller))
        (:CastFailed
          ;;[0 0 0 1 10] Special case for vehicles, as the driver pawn is sometimes unpossesed and does not have a player state
          (bind _asfort_player_pawn (Utilities|Casting|CastToFortPlayerPawn (LocalVariables|GetInActor)))
          (bind _asfort_vehicle_interface (Utilities|Casting|CastToFortVehicleInterface (Vehicle|GetVehicleActor _asfort_player_pawn :!class /Script/FortniteGame.FortPlayerPawn)))
          (Utilities|IsValid (Class|FortVehicleInterface|GetVehicleController _asfort_vehicle_interface _asfort_player_pawn)
            (:Is%20Valid
              (return (Class|FortVehicleInterface|GetVehicleController _asfort_vehicle_interface _asfort_player_pawn)))))))
    (:then_1
      (return (Fortnite|GetFortPlayerControllerfromActor (LocalVariables|GetInActor))))))

(fn GetTriggerLocation
  ;[1 1 1] Center the trigger location on the base
  (return (Math|Vector|MakeVector (.x (Variables|/Script/Engine.SceneComponent|GetRelativeLocation (Variables|Default|GetTrigger))) (.y (Variables|/Script/Engine.SceneComponent|GetRelativeLocation (Variables|Default|GetTrigger))) (+ (* (.z (Variables|/Script/Engine.BoxComponent|GetBoxExtent (Variables|Default|GetTrigger))) (Variables|Default|GetZoneZScale)) (Math|Vector|DotProduct (Variables|/Script/Engine.SceneComponent|GetRelativeLocation (Variables|Default|GetCreativeEditOnlyMesh)) (Variables|Default|GetTriggerLocationDotProduct))))))

(fn GetVisibilityComponents
  (bind (_base_components _zone_components) (Visibility|GetVisualComponents))
  (return _zone_components))

(fn GetVisualComponents
  ;[1 1 1] Set Default base and zone parts
  (return (Utilities|Array|MakeArray (Variables|Debug|GetStaticMeshComponent) (Variables|Default|GetLightRays)) (Utilities|Array|MakeArray (Variables|Default|GetZoneMesh))))

(fn GetVolumeMID
  (bind _creativeeditonlymesh (Variables|Default|GetCreativeEditOnlyMesh))
  (bind _asmaterial_instance_dynamic (Utilities|Casting|CastToMaterialInstanceDynamic (Rendering|Material|GetMaterial _creativeeditonlymesh 0))
    (:then
      (return _asmaterial_instance_dynamic))
    (:CastFailed
      (bind _returnvalue (Rendering|Material|CreateDynamicMaterialInstance _creativeeditonlymesh 0 "/Game/Abilities/Toys/EliminationVolume/M_EliminationVolume.M_EliminationVolume" "None" :!class /Script/Engine.PrimitiveComponent))
      (return _returnvalue))))

(fn GetZoneProtectedTeamValue
  (return (select (Knobs|Team|IsTeamNone (Variables|DeviceOptions|GetZoneProtectedTeam)) (Math|Conversions|ToInteger(Byte) (Utilities|Enum|LiteralenumEFortTeam "Spectator")) (Math|Conversions|ToInteger(Byte) (Teams|IndextoTeam (Knobs|Team|GetTeamValueasIndex (Variables|DeviceOptions|GetZoneProtectedTeam)))))))

(fn InitializeTeam
  ;[0.7362 1 0.3728] Handle Team Logic on setup
  (CallFunction|SetTeam (Math|Conversions|ToByte(Integer) (CallFunction|GetZoneProtectedTeamValue)))
  (TeamRestrictions|SetTeamswithAffects (Variables|Default|GetCreativeUseRestrictionComponent) (Variables|DeviceOptions|GetZoneProtectedTeam) (== (Variables|DeviceOptions|GetKnob_AffectsTeam) (Variables|/Game/Creative/Devices/Common/Components/Creative_UseRestriction_Component.Creative_UseRestriction_Component_C|GetAffectsAllButSelectedValue (Variables|Default|GetCreativeUseRestrictionComponent)))))

(fn InitializeClass
  ;[0.5102 1 0.2602] Ensure that we set the class so that it can be properly recognized by the use restriction component
  (Networking|SwitchHasAuthority
    (:Authority
      (ClassRestrictions|SetClassWithAffects (Variables|Default|GetCreativeUseRestrictionComponent) (Variables|DeviceOptions|GetKnob_SelectedClass) (== (Variables|DeviceOptions|GetKnob_AffectsClass) (Variables|/Game/Creative/Devices/Common/Components/Creative_UseRestriction_Component.Creative_UseRestriction_Component_C|GetAffectsAllButSelectedValue (Variables|Default|GetCreativeUseRestrictionComponent))) (Variables|Default|GetUses%20Safe%20Team%20Indexing)))))

(fn IsVisibleInGame
  (return (Variables|DeviceOptions|GetZoneVisibleInGame)))

(fn IsPlayerRegistered
  (params (PlayerToCheck :type Actor%20Object%20Reference%20%28by%20ref%29))
  ;[0.1969 0.1969 0.1969] What it says on the tin
  (return (Utilities|Array|ContainsItem (Variables|Default|GetAffectedPlayersInVolume) PlayerToCheck)))

(fn OnAffectedPlayerEnteredVehicle
  ;[1 0.75 0.5434] Edge-Case where we don't affect unattended vehicles but then it becomes attended. The ability handles target selection here, so we can fire it on entry safely.
  (Registration|RevalidateOverlappingPawns))

(fn OnAffectedPlayerExitedVehicle
  ;[1 1 1] Clear the GE from vehicles if there are no more affected pawns in vehicles, if the settings dictate
  (bind _output_get (LocalVariables|SetOverlappingTargets (Collision|GetOverlappingActors (Variables|Default|GetTrigger) 0 :!class /Script/Engine.PrimitiveComponent)))
  (for _array_element _output_get
    (bind _asfort_vehicle_interface (Utilities|Casting|CastToFortVehicleInterface _array_element))
    (if (and (== (Utilities|Array|Length (Vehicle|GetAllPassengers _asfort_vehicle_interface)) 0) (not false))
      (CallFunction|Non-PawnExitedTriggerVolume _array_element)
      (else
        ;;[0 0 0 1 10] Needed for when a player goes from driver to gunner in the Mech. When this elimination zone thinks the player has left the volume, but they are still in it.
        (CallFunction|Non-PawnEnteredTriggerVolume _array_element)))))

(fn PlayerOptionChangedDuringGame
  (params (Player%20State :type Fort%20Player%20State%20Object%20Reference))
  (bind _returnvalue (PlayerState|GetCurrentPawn Player%20State))
  ;[1 0.6906 0.7326] Handle Player Options Editing in Game
  (Networking|SwitchHasAuthority
    (:Authority
      ;;[1 1 1] If the device is disabled there is no overlap
      (if (Collision|IsOverlappingActor (Variables|Default|GetTrigger) _returnvalue :!class /Script/Engine.PrimitiveComponent)
        (Registration|RevalidatePawn _returnvalue)))))

(fn RegisterPlayerInVolume
  (params (PlayerToRegister :type Actor%20Object%20Reference%20%28by%20ref%29))
  ;[0.2719 0.2719 0.2719] A player is affected by volume. We register this to handle cases where player conditions change while affected, as those changes may causes their status with this volume to change (maybe they've become immune)
  (Utilities|IsValid PlayerToRegister
    (:Is%20Valid
      (bind _returnvalue (Utilities|Array|AddUnique (Variables|Default|GetAffectedPlayersInVolume) PlayerToRegister))
      (return _returnvalue))
    (:Is%20Not%20Valid
      (return -1))))

(fn RemoveAffectedPlayerInVolume
  (params (AffectedPlayer :type Actor%20Object%20Reference%20%28by%20ref%29))
  ;[1 1 1] A Player is no longer in the volume
  (bind _success (Registration|UnregisterPlayerinVolume AffectedPlayer))
  (if _success
    (Vehicles|UnbindVehicleDelegates AffectedPlayer)))

(fn ResizeFrame
  (bind _frame_mid (Variables|Default|GetFrame%20MID))
  (if (> (Rendering|Material|GetNumMaterials (Variables|Default|GetCreativeEditOnlyMesh)) 0)
    (Material|ReattachOrCreateMID (Variables|Default|GetCreativeEditOnlyMesh) _frame_mid)
    (Rendering|Material|SetVectorParameterValue _frame_mid "SizeBias" (Math|Conversions|ToLinearColor(Vector) (- (* (Variables|Default|GetMaterialScalar) (Math|Vector|MakeVector (Variables|Default|GetZoneXScale) (Variables|Default|GetZoneYScale) (Variables|Default|GetZoneZScale))) (Variables|Default|GetMaterialScalar))) :!class /Script/Engine.MaterialInstanceDynamic)
    (Rendering|SetBoundsScale (Variables|Default|GetCreativeEditOnlyMesh) (* (Math|Float|Max(Float) (Variables|Default|GetZoneXScale) (Variables|Default|GetZoneYScale) (Variables|Default|GetZoneZScale)) 1.5))))

(fn RevalidateOverlappingPawns
  ;[1 0.8806 0.6407] Check the pawns in our region. Has 3 outcomes that can be handled by each specific volume as needed. We only need player types, because non-players don't change their affectable options during gameplay (Team/Class)
  (bind _output_get (LocalVariables|SetCachedActors (Collision|GetOverlappingActors "/Script/FortniteGame.FortPawn")))
  (for _array_element _output_get
    (bind _asfort_pawn (Utilities|Casting|CastToFortPawn _array_element))
    (Utilities|FlowControl|Sequence
      (:then_0
        (CallFunction|PawnExitedTriggerVolume _asfort_pawn (Default|GetPlayerControllerfromActor _asfort_pawn)))
      (:then_1
        (bind (_valid_controller _fortplayerpawn _playercontroller) (DEPRECATED|ActorHasValidController (Variables|Default|GetCreativeUseRestrictionComponent) _asfort_pawn true true))
        (if _valid_controller
          (CallFunction|PawnEnteredTriggerVolume _fortplayerpawn _playercontroller))))))

(fn RevalidatePawn
  (params (PawnActor :type Actor%20Object%20Reference))
  (bind _playercontrollerfromactor (LocalVariables|GetPlayerControllerFromActor))
  (bind _local_playerpawn (LocalVariables|GetLocal_PlayerPawn))
  ;[0.4552 0.4552 0.4552] Circumstances changed, we need to recheck this pawn
  (Utilities|FlowControl|Sequence
    (:then_0
      (bind _asfort_pawn (Utilities|Casting|CastToFortPawn PawnActor))
      (LocalVariables|SetLocal_PlayerPawn _asfort_pawn)
      (LocalVariables|SetPlayerControllerFromActor (Default|GetPlayerControllerfromActor _asfort_pawn)))
    (:then_1
      (Default|ReProcessPawnValidation _local_playerpawn
        (:Retrigger
          (CallFunction|PawnEnteredTriggerVolume _local_playerpawn _playercontrollerfromactor))
        (:OverlapEnd
          (CallFunction|PawnExitedTriggerVolume _local_playerpawn _playercontrollerfromactor))))))

(fn SetBaseVisibilityAndCollision
  (params (BaseVisible :type Boolean))
  ;[1 0.8219 0] Dependent on base visibility bool, set the collision to either be enabled or disabled, if we're not in play mode force it to have colision enabled.
  (Visibility|SetVisiblityDuringPhase (Variables|Default|GetBaseVisibilityComponent) (select BaseVisible "NewEnumerator3" "NewEnumerator1"))
  (if (CallFunction|IsinPlayMode)
    (Collision|SetCollisionEnabled (Variables|Debug|GetStaticMeshComponent) (select BaseVisible "QueryAndPhysics" "NoCollision"))
    (else
      (Collision|SetCollisionEnabled (Variables|Debug|GetStaticMeshComponent) "QueryAndPhysics"))))

(fn ShouldShowInGame
  (params (DeviceOptionVisualInGame :type Boolean))
  ;[1 1 1] Determine if we are in a state where we care about visibility options
  (return (Math|Boolean|ANDBoolean (Class|CreativeEnabledComponent|IsEnabled (Variables|Default|GetEnabledComponent)) DeviceOptionVisualInGame (not (Replay|GetIsPlayingReplay)))))

(fn UnbindVehicleDelegates
  (params (Player :type Actor%20Object%20Reference))
  (bind _outputdelegate (EventDispatchers|CreateEvent :!event OnAffectedPlayerEnteredVehicle))
  ;[1 0.5137 0.4998] Remove special-case handling for affected players when no longer relevant
  (bind _output_get (LocalVariables|SetPlayerController (Default|GetPlayerControllerfromActor Player)))
  (Utilities|IsValid _output_get
    (:Is%20Valid
      (Vehicle|UnbindEventfromOnExitVehicle (LocalVariables|GetPlayerController) (EventDispatchers|CreateEvent :!event OnAffectedPlayerExitedVehicle))
      (Vehicles|OnAffectedPlayerExitedVehicle)
      (Vehicle|UnbindEventfromOnEnterVehicleDriver (LocalVariables|GetPlayerController) _outputdelegate)
      (Vehicle|UnbindEventfromOnEnterVehiclePassenger (LocalVariables|GetPlayerController) _outputdelegate))))

(fn UnregisterPlayerInVolume
  (params (PlayerToUnregister :type Actor%20Object%20Reference%20%28by%20ref%29))
  ;[0.2135 0.2135 0.2135] Remove a player from registration.
  (bind _returnvalue (Utilities|Array|RemoveItem (Variables|Default|GetAffectedPlayersInVolume) PlayerToUnregister))
  (return _returnvalue))

(fn UpdateBaseVisibilityComponent
  (bind _basevisibilitycomponent (Variables|Default|GetBaseVisibilityComponent)
    (:then
      (Initialization|InitializeWithSceneComponents _basevisibilitycomponent (Visibility|GetVisualComponents) (Utilities|Map|MakeMap) (select (Variables|DeviceOptions|GetBaseMeshVisibleInGame) "NewEnumerator3" "NewEnumerator1")))
    (:else)))

(fn UpdateVisiblity
  (bind (_base_visible _zone_visible) (Visibility|AreOptionalsVisibleDuringMinigame))
  ;[1 0.5895 0.8338] Set visibility based on User Options, but always hide the Frame when minigame is running
  (CallFunction|SetBaseVisibilityandCollision _base_visible)
  (Visibility|SetVisiblityDuringPhase (Variables|Default|GetZoneVisibilityComponent) (select _zone_visible "NewEnumerator3" "NewEnumerator1")))

(fn UpdateTriggerCollision
  (Collision|SetCollisionEnabled (Variables|Default|GetTrigger) (select (Class|CreativeEnabledComponent|IsEnabled (Variables|Default|GetEnabledComponent)) "QueryAndPhysics" "NoCollision"))
  (Networking|CanExecuteCosmeticEvents
    (:True
      (Collision|SetCollisionResponsetoChannel (Variables|Default|GetTrigger) "ECC_Pawn" (select (Variables|Default|GetbMinigameRunning) "ECR_Overlap" "ECR_Ignore")))))

(fn UpdateVolumeState
  ;[1 1 1] Check our collision state
  (Resizing|UpdateTriggerCollision)
  (Visibility|UpdateVisiblity))

(fn ConstructionScript
  (|CallParentFunction))
