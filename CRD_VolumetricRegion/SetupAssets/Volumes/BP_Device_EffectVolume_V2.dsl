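; Base Creative effect volume device that detects pawns, creatures, guard bots, vehicles and physics props entering or leaving a resizable zone, exposing empty enter/exit events.
; Overlaps from its own or an external volume are classified, filtered by team and class restrictions, and tracked in a set.
; Minigame start/end can bind delegates and equip an ability set.
; Spawns, team/class changes and vehicle entry revalidate pawns; enabled state drives collision and visibility.
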(event EventInitializeDevice
  ;[0 1 0.1413 0.25] Begin Play
  (|CallParentFunction)
  (CallFunction|GetFNEVolumeforOverlap)
  (Initialization|Initialize (Variables|Default|GetCreativeUseRestrictionComponent) (Variables|Default|GetFortMinigameProgress) (Variables|Default|GetEnabledComponent) 0 (Variables|UserOptions|GetSelectedTeam) (Variables|UserOptions|GetInvertTeamSelection) (Variables|UserOptions|GetSelectedClass) (Variables|UserOptions|GetInvertClassSelection) false)
  (Resizing|UpdateVolumeState)
  (CallFunction|UpdateVolumeSize))

(event Custom|EnableOnReceived
  (params (Instigator :type Controller%20Object%20Reference))
  ;[1 1 1] Channel-based activation logic
  (Class|CreativeEnabledComponent|SetEnabled (Variables|Default|GetEnabledComponent) true))

(event Custom|DisableOnReceived
  (params (Instigator :type Controller%20Object%20Reference))
  ;[1 1 1] Channel-based activation logic
  (for _array_element (Class|FNEVolumeOverlapComponent|GetAllTrackedActors (Variables|Default|GetFNE_VolumeOverlapComponent))
    (CallFunction|OnOverlap true _array_element))
  (Class|CreativeEnabledComponent|SetEnabled (Variables|Default|GetEnabledComponent) false))

(event Custom|Set%20Mesh
  ;[0 0.2775 2] Set Mesh
  (Class|FNEVolumeOverlapComponent|UpdateOverlapShape (Variables|Default|GetFNE_VolumeOverlapComponent) (Utilities|Select :Index (CallFunction|GetZoneShape) :NewEnumerator0 "BoxVolume" :NewEnumerator1 "BoxVolume" :NewEnumerator2 "BoxVolume" :NewEnumerator3 "SphereVolume" :NewEnumerator4 "CylinderVolume") false 0)
  (CallFunction|UpdateMaterials))

(event Custom|OnPlayerClassApplied
  (params (PlayerState :type Fort%20Player%20State%20Object%20Reference) (ClassSlotIndex :type Byte))
  ;[0.3833 0.4599 1] Handle Options Accessible During Gameplay
  (Registration|PlayerOptionChangedDuringGame PlayerState))

(event Custom|OnPlayerChangedTeam
  (params (PlayerState :type Fort%20Player%20State%20Object%20Reference) (TeamId :type Byte) (OldTeamId :type Byte) (bTeamChangedWithoutRespawn :type Boolean))
  ;[0.3833 0.4599 1] Handle Options Accessible During Gameplay
  (Registration|PlayerOptionChangedDuringGame PlayerState))

(event Custom|UpdateMaterials
  (bind (_value _returnvalue) (Utilities|Map|Find (Variables|Default|GetPreview%20Material) (CallFunction|GetZoneShape)))
  (bind _fne_volumeoverlapcomponent (Variables|Default|GetFNE_VolumeOverlapComponent))
  ;[0 0.2775 2] Set Mesh
  (if (or (Variables|UserOptions|GetZoneVisibleDuringGame) (not (CallFunction|IsinPlayMode)))
    (Class|FNEVolumeOverlapComponent|SetVolumeMaterial _fne_volumeoverlapcomponent (select (and _returnvalue (Building|IsPreviewBuildingActor)) _value (Variables|Default|GetVolume%20Material)) 0)
    (else
      (Class|FNEVolumeOverlapComponent|SetVolumeMaterial _fne_volumeoverlapcomponent "/Game/Missions/Secondary/BluGlo/M_Invisible.M_Invisible" 0))))

(event Custom|OnMinigamePlayerSpawned
  (params (SpawningPlayer :type Controller%20Object%20Reference))
  (bind _returnvalue (Pawn|GetControlledPawn SpawningPlayer))
  ;[0.3833 0.4599 1] Handle Options Accessible During Gameplay
  ;;[0.451 0.451 0.451 0.25] If the device is disabled, there is no overlap
  (Utilities|IsValid _returnvalue
    (:Is%20Valid
      (if (Collision|IsOverlappingActor (Variables|Default|GetActive%20FNE%20Volume%20Overlap%20Component) _returnvalue :!class /Script/Engine.PrimitiveComponent)
        (Registration|RevalidatePawn _returnvalue)))))

(event MinigameStartedDelegate(FortMinigameProgress)
  ;[0.9809 1 0.2776] Minigame State Changes (Begin and End, Transitionals)
  (if (Variables|Default|GetSetup%20at%20Minigame%20Start/End)
    (CallFunction|SetupDelegatesAndAbilities true)
    (Variables|Default|SetbMinigameRunning true)
    (Resizing|UpdateVolumeState)))

(event MinigameEndedDelegate(FortMinigameProgress)
  ;[0.9809 1 0.2776] Minigame State Changes (Begin and End, Transitionals)
  (if (Variables|Default|GetSetup%20at%20Minigame%20Start/End)
    (CallFunction|SetupDelegatesAndAbilities false)
    (Variables|Default|SetbMinigameRunning false)
    (Resizing|UpdateVolumeState)))

(event Custom|Setup%20Delegates%20and%20Abilities
  (params (Setup%20 :type Boolean))
  ;[0.9809 1 0.2776] Minigame State Changes (Begin and End, Transitionals)
  (if Setup%20
    (Networking|SwitchHasAuthority
      (:Authority
        (UserOptions|CustomInitializeClass (Creative|ClassOption|GetLegacyClassIndex (Variables|UserOptions|GetSelectedClass)))
        (UserOptions|CustomInitializeTeams (Creative|TeamOption|GetLegacyTeamIndex (Variables|UserOptions|GetSelectedTeam)))
        (bind _currentminigame (Default|GetCurrentMinigame)
          (:Is%20Valid
            (Minigame|BindEventtoOnPlayerClassApplied _currentminigame (EventDispatchers|CreateEvent :!event Custom|OnPlayerClassApplied))
            (Minigame|BindEventtoPlayerChangedTeamDelegate _currentminigame (EventDispatchers|CreateEvent :!event Custom|OnPlayerChangedTeam))
            (Minigame|AssignOnMinigamePlayerSpawned _currentminigame (EventDispatchers|CreateEvent :!event Custom|OnMinigamePlayerSpawned))
            (bind _returnvalue (Ability|EquipFortAbilitySet self (Variables|Ability|GetAbilitySystemReference)))
            (Variables|Ability|SetAbilitySetHandle _returnvalue)
            (CallFunction|InitializeDevice)
            (if (Variables|Default|GetUseExternalVolume)
              (CallFunction|BindtoExternalVolumeEvents))))))
    (else
      (Networking|SwitchHasAuthority
        (:Authority
          (bind _currentminigame_2 (Default|GetCurrentMinigame)
            (:Is%20Valid
              (Minigame|UnbindEventfromOnPlayerClassApplied _currentminigame_2 (EventDispatchers|CreateEvent :!event Custom|OnPlayerClassApplied))
              (Minigame|UnbindEventfromPlayerChangedTeamDelegate _currentminigame_2 (EventDispatchers|CreateEvent :!event Custom|OnPlayerChangedTeam))
              (Minigame|UnbindEventfromOnMinigamePlayerSpawned _currentminigame_2 (EventDispatchers|CreateEvent :!event Custom|OnMinigamePlayerSpawned))
              (Ability|UnequipFortAbilitySet (Variables|Ability|GetAbilitySetHandle))
              (if (Variables|Default|GetUseExternalVolume)
                (CallFunction|UnbindfromExternalEvents)))))))))

(event Custom|Pawn%20Entered%20Trigger%20Volume
  (params (Pawn :type Pawn%20Object%20Reference) (PlayerController :type Controller%20Object%20Reference)))

(event OnEnabledStateChanged(EnabledComponent)
  (params (Enabled :type Boolean))
  ;[1 0.3146 0.8498] On Enable
  (Resizing|UpdateVolumeState))

(event Custom|ExternalVolumeOverlapBegin
  (params (OtherActor :type Actor%20Object%20Reference))
  ;[0 0.787 1 0.2] Handle Overlap Events (Start and End)
  (CallFunction|OnOverlap false OtherActor))

(event OnOtherActorBeginOverlap(FNE_VolumeOverlapComponent)
  (params (OtherActor :type Actor%20Object%20Reference))
  ;[0 0.787 1 0.2] Handle Overlap Events (Start and End)
  (CallFunction|OnOverlap false OtherActor))

(event Custom|Pawn%20Exited%20Trigger%20Volume
  (params (Pawn :type Pawn%20Object%20Reference) (PlayerController :type Controller%20Object%20Reference)))

(event ChildFunctions|EventPlayModeChanged
  (params (Minigame :type Fort%20Minigame%20Object%20Reference) (bIsInPlayMode :type Boolean))
  (|CallParentFunction Minigame bIsInPlayMode)
  (CallFunction|UpdateMaterials))

(event Custom|Non-Pawn%20Entered%20Trigger%20Volume
  (params (Actor :type Actor%20Object%20Reference)))

(event Custom|ExternalVolumeOverlapEnd
  (params (OtherActor :type Actor%20Object%20Reference))
  ;[0 0.787 1 0.2] Handle Overlap Events (Start and End)
  (CallFunction|OnOverlap true OtherActor))

(event OnOtherActorEndOverlap(FNE_VolumeOverlapComponent)
  (params (OtherActor :type Actor%20Object%20Reference))
  ;[0 0.787 1 0.2] Handle Overlap Events (Start and End)
  (CallFunction|OnOverlap true OtherActor))

(event Custom|Non-Pawn%20Exited%20Trigger%20Volume
  (params (Actor :type Actor%20Object%20Reference)))

(event Custom|Player%20Bot%20Entered%20Trigger%20Volume
  (params (Pawn :type Pawn%20Object%20Reference) (AI%20Bot%20Controller :type Fort%20Athena%20AIBot%20Controller%20Object%20Reference)))

(event Custom|Player%20Bot%20Exited%20Trigger%20Volume
  (params (Pawn :type Pawn%20Object%20Reference) (AI%20Bot%20Controller :type Fort%20Athena%20AIBot%20Controller%20Object%20Reference)))

(event Custom|Physics%20Object%20Entered%20Trigger%20Volume
  (params (Actor :type Actor%20Object%20Reference)))

(event Custom|Physics%20Object%20Exited%20Trigger%20Volume
  (params (Actor :type Actor%20Object%20Reference)))

(event Custom|UpdateVolumeSize
  ;[0.6552 0.6552 0.6552] Handle Volume Resizing.  Assumes the volume is a mesh
  (bind _output_get (Variables|Default|SetNewScale (CallFunction|GetVolumeScale)))
  (Transformation|SetRelativeScale3D (Variables|Default|GetFNE_VolumeOverlapComponent) _output_get)
  (CallFunction|ResizeFrame)
  (CallFunction|SetMesh))

(event PawnNowHasValidParticipantController(CreativeUseRestrictionComponent)
  (params (FortPlayerPawn :type Fort%20Player%20Pawn%20Object%20Reference) (Controller :type Controller%20Object%20Reference))
  ;[1 1 1] Callback in case we are waiting on readiness
  (bind _result (Verse|IsAgentinVolume Controller))
  (if (and _result (CreativeUserRestrictions|CanParticipantControllerInteract (Variables|Default|GetCreativeUseRestrictionComponent) Controller true true false))
    (CallFunction|PawnEnteredTriggerVolume FortPlayerPawn Controller)))

(event Custom|BindToExternalVolumeEvents
  (bind _active_fne_volume_overlap_component (Variables|Default|GetActive%20FNE%20Volume%20Overlap%20Component))
  (Class|FNEVolumeOverlapComponent|SetEnableOverlapBehavior (Variables|Default|GetFNE_VolumeOverlapComponent) "Never")
  (Utilities|IsValid _active_fne_volume_overlap_component
    (:Is%20Valid
      (EventDispatchers|BindEventtoOnOtherActorBeginOverlap _active_fne_volume_overlap_component (EventDispatchers|CreateEvent :!event Custom|ExternalVolumeOverlapBegin))
      (EventDispatchers|BindEventtoOnOtherActorEndOverlap _active_fne_volume_overlap_component (EventDispatchers|CreateEvent :!event Custom|ExternalVolumeOverlapEnd)))))

(event Custom|UnbindFromExternalEvents
  (bind _active_fne_volume_overlap_component (Variables|Default|GetActive%20FNE%20Volume%20Overlap%20Component))
  (Class|FNEVolumeOverlapComponent|SetEnableOverlapBehavior (Variables|Default|GetFNE_VolumeOverlapComponent) "Game")
  (Utilities|IsValid _active_fne_volume_overlap_component
    (:Is%20Valid
      (EventDispatchers|UnbindEventfromOnOtherActorBeginOverlap _active_fne_volume_overlap_component (EventDispatchers|CreateEvent :!event Custom|ExternalVolumeOverlapBegin))
      (EventDispatchers|UnbindEventfromOnOtherActorEndOverlap _active_fne_volume_overlap_component (EventDispatchers|CreateEvent :!event Custom|ExternalVolumeOverlapEnd)))))

(fn ActorIsACreature
  (params (Actor :type Actor%20Object%20Reference))
  (Utilities|Casting|CastToFortPlayerPawn Actor
    (:then
      (return false 0))
    (:CastFailed
      (bind _asfort_pawn (Utilities|Casting|CastToFortPawn Actor)
        (:then
          (return true _asfort_pawn))
        (:CastFailed
          (return false 0))))))

(fn ActorIsAPlayerBot
  (params (Actor :type Actor%20Object%20Reference))
  (bind _returnvalue (AI|GetAIController Actor))
  ;[1 1 1] Check if the actor has an AI controller and then determine if it is AI Bot Controller
  (Utilities|IsValid _returnvalue
    (:Is%20Valid
      (bind _asfort_athena_aibot_controller (Utilities|Casting|CastToFortAthenaAIBotController _returnvalue)
        (:then
          (return true (Variables|/Script/FortniteAI.FortAthenaAIBotController|GetPlayerBotPawn _asfort_athena_aibot_controller) _asfort_athena_aibot_controller))
        (:CastFailed
          _)))
    (:Is%20Not%20Valid
      _))
  (return false 0 0))

(fn AddAffectedPlayerInVolume
  (params (AffectedPlayer :type Actor%20Object%20Reference%20%28by%20ref%29))
  ;[1 1 1] We've got a new player in the volume
  (Utilities|IsValid AffectedPlayer
    (:Is%20Valid
      (bind _updated (Registration|RegisterAgentinVolume AffectedPlayer))
      (if _updated
        (Vehicles|BindVehicleDelegates AffectedPlayer)))))

(fn AreOptionalsVisibleDuringMinigame
  ;[1 1 1] Get the current vis settings determined by users.
  (return (Visibility|ShouldShowinGame (Variables|UserOptions|GetZoneVisibleDuringGame))))

(fn BindVehicleDelegates
  (params (Player :type Actor%20Object%20Reference))
  (bind _outputdelegate (EventDispatchers|CreateEvent :!event OnAffectedPlayerEnteredVehicle))
  ;[0.5222 1 0.4526] Special-case handling for entering/exiting vehicles while affected by a volume
  (bind _asfort_player_controller (Utilities|Casting|CastToFortPlayerController (Default|GetPlayerControllerfromActor Player)))
  (bind _output_get (LocalVariables|SetPlayerController _asfort_player_controller))
  (Utilities|IsValid _output_get
    (:Is%20Valid
      (Vehicle|BindEventtoOnExitVehicle (LocalVariables|GetPlayerController) (EventDispatchers|CreateEvent :!event OnAffectedPlayerExitedVehicle))
      (Vehicle|BindEventtoOnEnterVehicleDriver (LocalVariables|GetPlayerController) _outputdelegate)
      (Vehicle|BindEventtoOnEnterVehiclePassenger (LocalVariables|GetPlayerController) _outputdelegate)
      (Vehicles|OnAffectedPlayerEnteredVehicle))))

(fn CustomInitializeClass
  (params (SelectedClass :type Integer))
  (Networking|SwitchHasAuthority
    (:Authority
      (ClassRestrictions|SetClassWithAffects (Variables|Default|GetCreativeUseRestrictionComponent) SelectedClass (Variables|UserOptions|GetInvertClassSelection) (Variables|Default|GetUses%20Safe%20Team%20Indexing)))))

(fn CustomInitializeTeams
  (params (SelectedTeam :type Integer))
  (Networking|SwitchHasAuthority
    (:Authority
      (TeamRestrictions|SetTeamswithAffects (Variables|Default|GetCreativeUseRestrictionComponent) SelectedTeam (Variables|UserOptions|GetInvertTeamSelection)))))

(fn GetAllVisualComponents
  ;[1 1 1] Set Default base and zone parts
  (return (Utilities|Array|MakeArray (Variables|Debug|GetStaticMeshComponent)) (Utilities|Array|MakeArray (Variables|Default|GetFNE_VolumeOverlapComponent))))

(fn GetEnabledStateasIndex
  (return (DeviceConversion|EnabledPhaseEnumtoLegacyInt (Variables|UserOptions|GetEnabledOnPhase))))

(fn GetFNEVolumeForOverlap
  (bind _externalvolume (Variables|UserOptions|Advanced|GetExternalVolume)
    (:then
      (bind _output_get (Variables|Default|SetExternal%20FNE%20Volume%20Overlap%20Component (Variables|/CRD_Volume/Device_Volume.Device_Volume_C|GetFNE_VolumeOverlapComponent _externalvolume))))
    (:else
      _))
  (bind _output_get_2 (Variables|Default|SetActive%20FNE%20Volume%20Overlap%20Component (select (Utilities|IsValid _output_get :!class /Script/Engine.KismetSystemLibrary) _output_get (Variables|Default|GetFNE_VolumeOverlapComponent))))
  (Variables|Default|SetUseExternalVolume (== _output_get_2 _output_get)))

(fn GetZoneShape
  (return "NewEnumerator1"))

(fn GetAlwaysEnableCollision
  (return false))

(fn GetCreativeActorBounds
  (bind (_origin _boxextent) (Collision|GetComponentBounds (Variables|Default|GetFNE_VolumeOverlapComponent)))
  (return true _boxextent))

(fn GetCreativeActorOrigin
  (return true (Math|Vector|MakeVector 0.0 0.0 (* (Variables|UserOptions|GetZoneHeight) (Variables|Default|GetPivotOriginOffset)))))

(fn GetCreativeActorRelativeTransform
  (return false 0))

(fn GetGridAxisGuideAttachmentPoint
  (return false false "0, 0, 0"))

(fn GetOccupiers
  ;[0.9594 0 1 0.5] Returns any agents in this volume.
  (bind _result (Utilities|Set|ToArray (Variables|Default|GetAffectedPlayersInVolume)))
  (for _array_element _result
    (Utilities|IsValid (Fortnite|GetControllerfromActor _array_element)
      (:Is%20Valid
        (Utilities|Array|Add (LocalVariables|GetOccupyingControllers) (Fortnite|GetControllerfromActor _array_element)))))
  (return (LocalVariables|GetOccupyingControllers)))

(fn GetOverrideMeshMaterial
  (params (MeshComp :type Static%20Mesh%20Component%20Object%20Reference) (MatIdx :type Integer))
  (if (Utilities|Equal(Object) MeshComp (Variables|Default|GetCreativeVolumeFrame))
    (return (Variables|Default|GetFrame%20MID))
    (else
      (return (|CallParentFunction MeshComp MatIdx)))))

(fn GetPlayerControllerFromActor
  (params (Actor :type Actor%20Object%20Reference))
  ;[1 0.7271 0.6488] Retrieve the player controller from an actor.  Assumes any input so might not be the best option in all cases.
  (LocalVariables|SetInActor Actor)
  (Utilities|FlowControl|Sequence
    (:then_0
      (bind _aspawn (Utilities|Casting|CastToPawn (LocalVariables|GetInActor)))
      (bind _asfort_player_controller (Utilities|Casting|CastToFortPlayerController (Pawn|GetController _aspawn))
        (:then
          (return _asfort_player_controller))
        (:CastFailed
          ;;[0 0 0 1 10] Special case for vehicles, as the driver pawn is sometimes unpossesed and does not have a player state
          (bind _asfort_player_pawn (Utilities|Casting|CastToFortPlayerPawn (LocalVariables|GetInActor)))
          (bind _asfort_vehicle_interface (Utilities|Casting|CastToFortVehicleInterface (Vehicle|GetVehicleActor _asfort_player_pawn :!class /Script/FortniteGame.FortPlayerPawn)))
          (Utilities|IsValid (Class|FortVehicleInterface|GetVehicleController _asfort_vehicle_interface _asfort_player_pawn)
            (:Is%20Valid
              (return (Class|FortVehicleInterface|GetVehicleController _asfort_vehicle_interface _asfort_player_pawn)))))))
    (:then_1
      (return (CreativeAI|Controller|GetParticipantControllerfromActor (LocalVariables|GetInActor))))))

(fn GetTriggerLocation
  (bind (_returnvalue_x _returnvalue_y _returnvalue_z) (Math|Vector|BreakVector (CallFunction|GetVolumeScale)))
  (bind (_origin _boxextent) (Collision|GetComponentBounds (Variables|Default|GetFNE_VolumeOverlapComponent)))
  ;[1 1 1] Center the trigger location on the base
  (return (Math|Vector|MakeVector (.x (Variables|/Script/Engine.SceneComponent|GetRelativeLocation (Variables|Default|GetFNE_VolumeOverlapComponent))) (.y (Variables|/Script/Engine.SceneComponent|GetRelativeLocation (Variables|Default|GetFNE_VolumeOverlapComponent))) (+ (* (.z _boxextent) _returnvalue_z) (Math|Vector|DotProduct (Variables|/Script/Engine.SceneComponent|GetRelativeLocation (Variables|Default|GetCreativeVolumeFrame)) (Variables|Default|GetTriggerLocationDotProduct))))))

(fn GetVisibilityComponents
  (bind (_base_components _zone_components) (Visibility|GetAllVisualComponents))
  (return _zone_components))

(fn GetVolumeScale
  (return (Math|Vector|MakeVector (Math|Float|Max(Float) (Variables|UserOptions|GetZoneWidth) 0.05) (Math|Float|Max(Float) (Variables|UserOptions|GetZoneDepth) 0.05) (Math|Float|Max(Float) (Variables|UserOptions|GetZoneHeight) 0.05))))

(fn IsActorRegistered
  (params (ActorToCheck :type Actor%20Object%20Reference%20%28by%20ref%29))
  ;[0.1969 0.1969 0.1969] What it says on the tin
  (return (Utilities|Set|ContainsItem (Variables|Default|GetAffectedPlayersInVolume) ActorToCheck)))

(fn IsVisibleInGame
  (return (Variables|UserOptions|GetZoneVisibleDuringGame)))

(fn IsAgentInVolume
  (params (Controller :type Controller%20Object%20Reference))
  (bind _returnvalue (Pawn|GetControlledPawn Controller))
  (Utilities|IsValid Controller
    (:Is%20Valid
      (Utilities|IsValid _returnvalue
        (:Is%20Valid
          (return (Registration|IsActorRegistered _returnvalue)))
        (:Is%20Not%20Valid
          _)))
    (:Is%20Not%20Valid
      _))
  (return false))

(fn MigrateLegacyUserOptions
  (|CallParentFunction)
  (Utilities|FlowControl|Sequence
    (:then_0
      ;[1 0 0.623 0.2] Converting Team and Class to remove the concept of 'None' and replace with 'Any' (for Team, this is a straight conversion. For Class, None = !Any, so we flip the Invert Class Selection option.
      (Utilities|FlowControl|Sequence
        (:then_0
          (if (== (Creative|TeamOption|GetLegacyTeamIndex (Variables|UserOptions|GetSelectedTeam)) 999)
            (Variables|SetSelectedTeam_TeamType (struct :TeamType "Any" :TeamIndex 1)) ; In Old Method, None=Any
            (Variables|UserOptions|SetInvertTeamSelection false) ; If it was set to None, invert team didn't matter, \nso we explicitly set it to false as it now does matter.
            (LocalVariables|SetOptionsMigrated true)))
        (:then_1
          (if (== (Creative|ClassOption|GetLegacyClassIndex (Variables|UserOptions|GetSelectedClass)) -2)
            (Variables|SetSelectedClass_ClassType (struct :ClassType "Any" :ClassSlot 1))
            (Variables|UserOptions|SetInvertClassSelection (not (Variables|UserOptions|GetInvertClassSelection)))
            (LocalVariables|SetOptionsMigrated true)))))
    (:then_1
      (if (LocalVariables|GetOptionsMigrated)
        (OptionData|RefreshPropertyOverrides (Variables|Default|GetToyOptionsComponent))))))

(fn OnOverlap
  (params (Has%20Ended%20Overlap :type Boolean) (Other%20Actor :type Actor%20Object%20Reference))
  (bind _other_actor (Variables|GetOther%20Actor))
  (bind (_true _pawn) (Default|ActorIsACreature _other_actor))
  (bind _other_actor_2 (Variables|GetOther%20Actor))
  (bind (_true_2 _pawn_2 _ai_bot_controller) (Default|ActorIsAPlayerBot _other_actor_2))
  (bind _other_actor_3 (Variables|GetOther%20Actor))
  (bind _other_actor_4 (Variables|GetOther%20Actor))
  (if (not (Creative|UEFN|IsUEFNEditModeTransitionPending 0))
    (Utilities|FlowControl|Sequence
      (:then_0
        (Utilities|Casting|CastToFortVehicleInterface (Variables|GetOther%20Actor)
          (:then
            (if (Variables|GetHas%20Ended%20Overlap)
              (CallFunction|Non-PawnExitedTriggerVolume _other_actor_3)
              (else
                (CallFunction|Non-PawnEnteredTriggerVolume _other_actor_3))))
          (:CastFailed
            (if (Math|Boolean|ANDBoolean (Utilities|IsValid (Actor|GetComponentbyClass (Variables|GetOther%20Actor) "/Script/FortniteGame.FortPhysicsComponent") :!class /Script/Engine.KismetSystemLibrary) (Fortnite|IsPhysicallySimulatedIsland) (Variables|UserOptions|GetAffectsPhysicsProps))
              (if (Variables|GetHas%20Ended%20Overlap)
                (CallFunction|PhysicsObjectExitedTriggerVolume _other_actor_4)
                (else
                  (CallFunction|PhysicsObjectEnteredTriggerVolume _other_actor_4)))
              (else
                (bind _updated (Registration|RegisterAgentinVolume Other%20Actor (not Has%20Ended%20Overlap)))
                (if _updated
                  (bind (_valid_controller _fortplayerpawn _controller) (CreativeUserRestrictions|ActorHasValidParticipantController (Variables|Default|GetCreativeUseRestrictionComponent) (Variables|GetOther%20Actor) true true true))
                  (if _valid_controller
                    (if _true_2
                      (bind _isguard (CreativeAI|NPCType|IsGuardActor (Variables|GetOther%20Actor)))
                      (if _isguard
                        (if (Variables|GetHas%20Ended%20Overlap)
                          (CallFunction|PlayerBotExitedTriggerVolume _pawn_2 _ai_bot_controller)
                          (else
                            (CallFunction|PlayerBotEnteredTriggerVolume _pawn_2 _ai_bot_controller)))
                        (else
                          _))
                      (else
                        _))
                    (if (Variables|GetHas%20Ended%20Overlap)
                      (CallFunction|PawnExitedTriggerVolume _fortplayerpawn _controller)
                      (else
                        (CallFunction|PawnEnteredTriggerVolume _fortplayerpawn _controller)))
                    (elif _true
                      (if (Variables|GetHas%20Ended%20Overlap)
                        (CallFunction|PawnExitedTriggerVolume _pawn)
                        (else
                          (CallFunction|PawnEnteredTriggerVolume _pawn)))
                      (else
                        (return false))))
                  (else
                    (return false))))))))
      (:then_1
        (return true)))))

(fn OnAffectedPlayerEnteredVehicle
  ;[1 0.75 0.5434] Edge-Case where we don't affect unattended vehicles but then it becomes attended. The ability handles target selection here, so we can fire it on entry safely.
  (Registration|RevalidateOverlappingPawns))

(fn OnAffectedPlayerExitedVehicle
  ;[1 1 1] Clear the GE from vehicles if there are no more affected pawns in vehicles, if the settings dictate
  (bind _output_get (LocalVariables|SetOverlappingTargets (Collision|GetOverlappingActors (Variables|Default|GetFNE_VolumeOverlapComponent) 0 :!class /Script/Engine.PrimitiveComponent)))
  (for _array_element _output_get
    (bind _asfort_vehicle_interface (Utilities|Casting|CastToFortVehicleInterface _array_element))
    (if (and (== (Utilities|Array|Length (Vehicle|GetAllPassengers _asfort_vehicle_interface)) 0) (not false))
      (CallFunction|Non-PawnExitedTriggerVolume _array_element)
      (else
        ;;[0 0 0 1 10] Needed for when a player goes from driver to gunner in the Mech. When this elimination zone thinks the player has left the volume, but they are still in it.
        (CallFunction|Non-PawnEnteredTriggerVolume _array_element)))))

(fn PlayerOptionChangedDuringGame
  (params (Player%20State :type Fort%20Player%20State%20Object%20Reference))
  (bind _returnvalue (PlayerState|GetCurrentPawn Player%20State))
  ;[1 0.6906 0.7326] Handle Player Options Editing in Game
  (Networking|SwitchHasAuthority
    (:Authority
      ;;[1 1 1] If the device is disabled there is no overlap
      (if (Collision|IsOverlappingActor (Variables|Default|GetFNE_VolumeOverlapComponent) _returnvalue :!class /Script/Engine.PrimitiveComponent)
        (Registration|RevalidatePawn _returnvalue)))))

(fn RegisterAgentInVolume
  (params (AgentToRegister :type Actor%20Object%20Reference%20%28by%20ref%29) (Add :type Boolean :default true))
  (bind _affectedplayersinvolume (Variables|Default|GetAffectedPlayersInVolume))
  ;[0.2719 0.2719 0.2719] A player is affected by volume. We register this to handle cases where player conditions change while affected, as those changes may causes their status with this volume to change (maybe they've become immune)
  (Utilities|Casting|CastToFortPawn AgentToRegister
    (:then
      (if Add
        (if (Utilities|Set|ContainsItem (Variables|Default|GetAffectedPlayersInVolume) AgentToRegister)
          (return false)
          (else
            (Utilities|Set|Add _affectedplayersinvolume AgentToRegister)
            (return true)))
        (else
          (bind _returnvalue (Utilities|Set|Remove _affectedplayersinvolume AgentToRegister))
          (return _returnvalue))))
    (:CastFailed
      (return false))))

(fn RemoveAffectedPlayerInVolume
  (params (AffectedPlayer :type Actor%20Object%20Reference%20%28by%20ref%29))
  ;[1 1 1] A Player is no longer in the volume
  (bind _updated (Registration|RegisterAgentinVolume AffectedPlayer false))
  (if _updated
    (Vehicles|UnbindVehicleDelegates AffectedPlayer)))

(fn ResizeFrame
  (Class|CreativeVolumeFrame|InitializeComponent (Variables|Default|GetCreativeVolumeFrame) (CallFunction|GetZoneShape) 1.0 (Utilities|Operators|Multiply (CallFunction|GetVolumeScale) (5.12 5.12 3.84))))

(fn RevalidateOverlappingPawns
  ;[1 0.8806 0.6407] Check the pawns in our region. Has 3 outcomes that can be handled by each specific volume as needed. We only need player types, because non-players don't change their affectable options during gameplay (Team/Class)
  (bind _output_get (LocalVariables|SetCachedActors (Collision|GetOverlappingActors "/Script/FortniteGame.FortPawn")))
  (for _array_element _output_get
    (bind _asfort_pawn (Utilities|Casting|CastToFortPawn _array_element))
    (Utilities|FlowControl|Sequence
      (:then_0
        (CallFunction|PawnExitedTriggerVolume _asfort_pawn (Default|GetPlayerControllerfromActor _asfort_pawn)))
      (:then_1
        (bind (_valid_controller _fortplayerpawn _controller) (CreativeUserRestrictions|ActorHasValidParticipantController (Variables|Default|GetCreativeUseRestrictionComponent) _asfort_pawn true true true))
        (if _valid_controller
          (CallFunction|PawnEnteredTriggerVolume _fortplayerpawn _controller))))))

(fn RevalidatePawn
  (params (PawnActor :type Actor%20Object%20Reference))
  (bind _playercontrollerfromactor (LocalVariables|GetPlayerControllerFromActor))
  (bind _local_playerpawn (LocalVariables|GetLocal_PlayerPawn))
  ;[0.4552 0.4552 0.4552] Circumstances changed, we need to recheck this pawn
  (Utilities|FlowControl|Sequence
    (:then_0
      (bind _asfort_pawn (Utilities|Casting|CastToFortPawn PawnActor))
      (LocalVariables|SetLocal_PlayerPawn _asfort_pawn)
      (LocalVariables|SetPlayerControllerFromActor (Default|GetPlayerControllerfromActor _asfort_pawn)))
    (:then_1
      (Default|ReProcessPawnValidation _local_playerpawn
        (:Retrigger
          (CallFunction|PawnEnteredTriggerVolume _local_playerpawn _playercontrollerfromactor))
        (:OverlapEnd
          (CallFunction|PawnExitedTriggerVolume _local_playerpawn _playercontrollerfromactor))))))

(fn ShouldShowInGame
  (params (DeviceOptionVisualInGame :type Boolean))
  ;[1 1 1] Determine if we are in a state where we care about visibility options
  (return (Math|Boolean|ANDBoolean (Class|CreativeEnabledComponent|IsEnabled (Variables|Default|GetEnabledComponent)) DeviceOptionVisualInGame (not (Replay|GetIsPlayingReplay)))))

(fn UnbindVehicleDelegates
  (params (Player :type Actor%20Object%20Reference))
  (bind _outputdelegate (EventDispatchers|CreateEvent :!event OnAffectedPlayerEnteredVehicle))
  ;[1 0.5137 0.4998] Remove special-case handling for affected players when no longer relevant
  (bind _asfort_player_controller (Utilities|Casting|CastToFortPlayerController (Default|GetPlayerControllerfromActor Player)))
  (bind _output_get (LocalVariables|SetPlayerController _asfort_player_controller))
  (Utilities|IsValid _output_get
    (:Is%20Valid
      (Vehicle|UnbindEventfromOnExitVehicle (LocalVariables|GetPlayerController) (EventDispatchers|CreateEvent :!event OnAffectedPlayerExitedVehicle))
      (Vehicles|OnAffectedPlayerExitedVehicle)
      (Vehicle|UnbindEventfromOnEnterVehicleDriver (LocalVariables|GetPlayerController) _outputdelegate)
      (Vehicle|UnbindEventfromOnEnterVehiclePassenger (LocalVariables|GetPlayerController) _outputdelegate))))

(fn UpdateVisiblity
  ;[1 0.5895 0.8338] Set visibility based on User Options, but always hide the Frame when minigame is running
  (Networking|SwitchHasAuthority
    (:Authority
      (if (CallFunction|IsinPlayMode)
        (Visibility|SetForcedVisibilityDuringGame (Variables|Default|GetVisibleInGameComponent) (select (Visibility|AreOptionalsVisibleDuringMinigame) "NewEnumerator1" "NewEnumerator2"))
        (else
          (Visibility|SetForcedVisibilityDuringGame (Variables|Default|GetVisibleInGameComponent) "NewEnumerator1"))))))

(fn UpdateTriggerCollision
  (bind _fne_volumeoverlapcomponent (Variables|Default|GetFNE_VolumeOverlapComponent))
  (Networking|SwitchHasAuthority
    (:Authority
      (Collision|SetCollisionEnabled (Variables|Default|GetFNE_VolumeOverlapComponent) (select (or (Class|CreativeEnabledComponent|IsEnabled (Variables|Default|GetEnabledComponent)) (CallFunction|GetAlwaysEnableCollision)) "QueryAndPhysics" "NoCollision"))
      (if (not (Editor|InEditor))
        (if (not (Utilities|Enum|Equal(Enum) (Variables|UserOptions|GetEnabledOnPhase) "NewEnumerator0"))
          (Collision|SetCollisionResponsetoChannel _fne_volumeoverlapcomponent "ECC_Pawn" (select (Variables|Default|GetbMinigameRunning) "ECR_Overlap" "ECR_Ignore"))
          (else
            _))
        (else
          _))
      (Collision|SetCollisionResponsetoChannel _fne_volumeoverlapcomponent "ECC_Pawn" "ECR_Overlap"))))

(fn UpdateVolumeState
  ;[1 1 1] Check our collision state
  (Resizing|UpdateTriggerCollision)
  (Visibility|UpdateVisiblity))

(fn ConstructionScript
  (bind _fne_volumeoverlapcomponent (Variables|Default|GetFNE_VolumeOverlapComponent))
  (|CallParentFunction)
  ;[1 1 1] Force Collision Object Type and Enabled to match with Component default
  (Collision|SetCollisionObjectType _fne_volumeoverlapcomponent "ECC_WorldDynamic")
  (Collision|SetCollisionEnabled _fne_volumeoverlapcomponent "QueryOnly")
  (if (not (Editor|InEditor))
    (Rendering|Material|SetMaterial (Variables|Default|GetFNE_VolumeOverlapComponent) 0 0) ; Attempting to fix FORT-683185, reapplying invisible material\nearly in the process to reset any material that was saved on\nthe component instance.
  ))
